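# Super Level 24 material function that restyles a LocalZilla player character's incoming material attributes and also outputs a boxes mask.
# A mask texture splits the surface into Primary, Secondary, Accent and Paint regions, each with its own colors and surface parameters.
# Primary layers fresnel, wireframe, scanlines and animated noise; Secondary adds camera-facing virtual-plane distance-field boxes.
# Static switches select a banded T3 pattern, toon shading with stylized reflections, or an inferno recolor.
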MaterialFunction /Game/Characters/Player/Male/Medium/Bodies/M_MED_LocalZilla/Materials/Super_Levels/SL24_Functions/MFC_S24_SuperLevel
graph {nodes [
  Fresnel() {baseReflectFraction=0, exponent=1}
  FunctionOutput(None=NamedRerouteUsage())
  FunctionOutput(None=NamedRerouteUsage()) {outputName=Boxes Mask, sortPriority=20}
  n1 = Multiply(A=NamedRerouteUsage(), B=NamedRerouteUsage())
  Multiply(A=@n1, B=OneMinus(None=NamedRerouteUsage()))
  n2 = Multiply(
   A=VectorParameter() {group=Super Level 24, parameterName=Pnt_OverrideColor, defaultValue=<rgba 1 1 1 0>}.RGB
   B=VectorParameter() {group=Super Level 24, parameterName=Pnt_Color, defaultValue=<rgba 0.01 0.01 0.01 0>}.RGB
  )
  Multiply(A=NamedRerouteUsage(), B=@n2)
  NamedRerouteDeclaration(None=FunctionInput() {bUsePreviewValueAsDefault=true, inputType=FunctionInput_MaterialAttributes}) {name=SuperLevel, nodeColor=<rgba 0.2188 0 1 1>}
  n3 = NamedRerouteUsage()
  n4 = NamedRerouteUsage()
  n5 = GetMaterialAttributes(None=NamedRerouteUsage()) {attributeGetTypes [
    #1
    #2
    #3
    #4
    #5
    #6
    #7
   ]}
  n6 = SetMaterialAttributes!(
   MaterialAttributes=@n5.MaterialAttributes
   BaseColor=NamedRerouteUsage()
   Metallic=NamedRerouteUsage()
   Specular=NamedRerouteUsage()
   Roughness=NamedRerouteUsage()
   EmissiveColor=NamedRerouteUsage()
   Normal=NamedRerouteUsage()
   OpacityMask
  ) {attributeSetTypes [
    #1
    #2
    #3
    #4
    #5
    #6
    #7
   ]}
  NamedRerouteDeclaration(None=StaticSwitchParameter(
    True=SetMaterialAttributes!(MaterialAttributes=@n5.MaterialAttributes, EmissiveColor=@n3, Normal=@n4) {attributeSetTypes [#5, #6]}
    False=@n6
   ) {parameterName=Use Toon Shading, sortPriority=900}) {name=SuperLevelOut, nodeColor=<rgba 0.2188 0 1 1>}
  NamedRerouteDeclaration(None=@n5.Specular) {name=Specular In, nodeColor=<rgba 0.0078 0 1 1>}
  NamedRerouteDeclaration(None=@n5.Roughness) {name=Roughness In, nodeColor=<rgba 0.0078 0 1 1>}
  n7 = NamedRerouteUsage()
  n8 = NamedRerouteUsage()
  n9 = NamedRerouteUsage()
  n10 = NamedRerouteUsage()
  n11 = LinearInterpolate(A=@n10, B=NamedRerouteUsage(), Alpha=ScalarParameter() {defaultValue=0.01, parameterName=Pri_BlendPixels}) {constAlpha=0.05}
  n12 = LinearInterpolate(
   A=LinearInterpolate(
    A=LinearInterpolate(A=@n7, B=@n8, Alpha=@n9)
    B=StaticSwitchParameter(True=@n10, False=@n11) {group=Super Level 24, parameterName=UseT3Pattern, sortPriority=3000}
    Alpha=NamedRerouteUsage()
   )
   B=NamedRerouteUsage()
   Alpha=NamedRerouteUsage()
  )
  n13 = NamedRerouteUsage()
  n14 = NamedRerouteUsage()
  NamedRerouteDeclaration(None=NamedRerouteDeclaration(None=LinearInterpolate(A=@n12, B=@n13, Alpha=@n14)) {name=Base Color Toon, nodeColor=<rgba 0 1 0.5313 1>}) {name=Base Color Out, nodeColor=<rgba 1 0.8906 0 1>}
  n15 = LinearInterpolate(A=NamedRerouteUsage(), B=NamedRerouteUsage(), Alpha=NamedRerouteUsage())
  n16 = StaticSwitchParameter(True=LinearInterpolate(A=@n15, B=ScalarParameter() {defaultValue=1, parameterName=Sec_T3_Metallic}, Alpha=NamedRerouteUsage()), False=@n15) {group=Super Level 24, parameterName=UseT3Pattern, sortPriority=3000}
  n17 = NamedRerouteUsage()
  n18 = NamedRerouteUsage()
  NamedRerouteDeclaration(None=StaticSwitchParameter(True=Constant(), False=LinearInterpolate(A=@n16, B=@n17, Alpha=@n18)) {group=Super Level Debug, parameterName=Use Debug Mode, sortPriority=0}) {name=Metallic Out, nodeColor=<rgba 1 0.8906 0 1>}
  NamedRerouteDeclaration(None=StaticSwitchParameter(True=ScalarParameter() {defaultValue=0.35, parameterName=Pri_T3_Specular}, False=NamedRerouteUsage()) {group=Super Level 24, parameterName=UseT3Pattern, sortPriority=3000}) {name=Specular Out, nodeColor=<rgba 1 0.8906 0 1>}
  n19 = LinearInterpolate(A=NamedRerouteUsage(), B=NamedRerouteUsage(), Alpha=NamedRerouteUsage())
  n20 = Multiply(A=NamedRerouteUsage(), B=ScalarParameter() {defaultValue=1, group=Super Level 24, parameterName=Sec_Roughness})
  n21 = NamedRerouteUsage()
  NamedRerouteDeclaration(None=LinearInterpolate(A=LinearInterpolate(A=@n19, B=@n20, Alpha=@n21), B=NamedRerouteUsage(), Alpha=NamedRerouteUsage())) {name=Roughness Out, nodeColor=<rgba 1 0.8906 0 1>}
  n22 = NamedRerouteUsage()
  n23 = VectorParameter() {group=Super Level 24, parameterName=Pri_ColorGradientLocation, defaultValue=<rgba 0 0 100 1>}
  n24 = ScalarParameter() {defaultValue=50, group=Super Level 24, parameterName=Pri_ColorGradientRadius}
  n25 = ScalarParameter() {group=Super Level 24, parameterName=Pri_ColorGradientHardness}
  n26 = VectorParameter() {group=Super Level 24, parameterName=Pri_BaseColorFresnel, defaultValue=<rgba 0.0324 0 0.2158 1>}
  n27 = VectorParameter() {group=Super Level 24, parameterName=Pri_BaseColorFresnelNight, defaultValue=<rgba 1 0.05 0.0817 1>}
  n28 = Multiply(A=NamedRerouteUsage(), B=NamedRerouteUsage())
  n29 = Max(A=@n28, B=NamedRerouteUsage())
  n30 = LinearInterpolate(
   A=LinearInterpolate(
    A=VectorParameter() {group=Super Level 24, parameterName=Pri_BaseColor, defaultValue=<rgba 0.02 0.0001 0.03 1>}.RGB
    B=VectorParameter() {group=Super Level 24, parameterName=Pri_BaseColorFalloff, defaultValue=<rgba 0.0223 0.0002 0.0333 1>}.RGB
    Alpha=SphereMask(A=@n22, B=@n23.RGB, Radius=@n24, Hardness=@n25)
   )
   B=StaticSwitchParameter(
    True=MaterialFunctionCall(Color_Morning=@n26.RGB, Color_Day=@n26.RGB, Color_Evening=@n26.RGB, Color_Night=@n27.RGB) {materialFunction=&Color_Time_of_Day}.Result
    False=@n26.RGB
   ) {group=Super Level 24, parameterName=UseT3Pattern, sortPriority=3000}
   Alpha=StaticSwitchParameter(
    True=StaticSwitchParameter(True=@n28, False=@n29) {group=Super Level 24 Masks, parameterName=DisablePrimaryFX, sortPriority=3000}
    False=@n28
   ) {group=Super Level 24, parameterName=UseT3Pattern, sortPriority=3000}
  )
  n31 = ScalarParameter() {defaultValue=0.35, parameterName=Pri_T3_ScanlinesMin}
  n32 = NamedRerouteUsage()
  n33 = NamedRerouteUsage()
  n34 = NamedRerouteUsage()
  n35 = LinearInterpolate(
   A=@n30
   B=Multiply(
    A=VectorParameter() {group=Super Level 24, parameterName=Pri_BandsColor, defaultValue=<rgba 0.9427 0.3142 0 1>}.RGB
    B=LinearInterpolate(A=@n31, Alpha=@n32)
   )
   Alpha=NamedRerouteDeclaration(None=Max(A=@n33, B=@n34)) {name=T3 Gold Trim, nodeColor=<rgba 0.9453 1 0 1>}
  )
  n36 = Add(A=LinearInterpolate(A=@n30, B=NamedRerouteUsage(), Alpha=NamedRerouteUsage()) {desc=12 instructions}, B=NamedRerouteUsage())
  NamedRerouteDeclaration(None=StaticSwitchParameter(True=@n30, False=StaticSwitchParameter(True=@n35, False=@n36) {group=Super Level 24, parameterName=UseT3Pattern, sortPriority=3000}) {group=Super Level 24 Masks, parameterName=DisablePrimaryFX, sortPriority=3000}) {name=Primary Color, nodeColor=<rgba 1 0 0.0496 1>}
  NamedRerouteDeclaration(None=@n2) {name=Paint Out, nodeColor=<rgba 0.4757 0.4315 0.5 1>}
  n37 = NamedRerouteUsage()
  n38 = ScalarParameter() {defaultValue=0.001, parameterName=NoiseMult}
  n39 = TextureObjectParameter() {parameterName=Wireframe Texture, samplerType=SAMPLERTYPE_Masks, texture=&T_BlackMask}
  n40 = ScalarParameter() {defaultValue=-3.5, parameterName=Pri_WireMipBias}
  n41 = TextureSample!(UVs=Add(A=NamedRerouteUsage(), B=Multiply(A=@n37, B=@n38)), Tex=@n39, Bias=@n40, Apply View MipBias) {mipValueMode=TMVM_MipBias}
  NamedRerouteDeclaration(None=@n41.B) {name=Blue Channel Extras, nodeColor=<rgba 0.0078 0 1 1>}
  n42 = VectorParameter() {group=Super Level 24, parameterName=Acc_OverrideColor, defaultValue=<rgba 1 1 1 0>}
  n43 = Desaturation(None=NamedRerouteUsage(), Fraction=ScalarParameter() {defaultValue=1, group=Super Level 24, parameterName=Acc_Desdaturation})
  n44 = ScalarParameter() {group=Super Level 24, parameterName=Acc_UseBaseColor}
  n45 = VectorParameter() {group=Super Level 24, parameterName=Acc_Color, defaultValue=<rgba 0.011 0.2016 0.5841 1>}
  n46 = NamedRerouteUsage()
  NamedRerouteDeclaration(None=Multiply(A=LinearInterpolate(A=@n42.RGB, B=@n43, Alpha=@n44), B=Multiply(A=@n45.RGB, B=@n46))) {name=Accent Out, nodeColor=<rgba 0.0532 1 0 1>}
  n47 = TextureSampleParameter2D() {
   group=Super Level 24
   parameterName=Super Level Mask
   samplerType=SAMPLERTYPE_Masks
   texture=&T_BlackMask
  }
  n48 = Max(A=Max(A=@n47.R, B=@n47.R), B=@n47.B)
  n49 = OneMinus(None=NamedRerouteUsage())
  n50 = ScalarParameter() {parameterName=SL_BypassSkinCheck}
  n51 = AppendVector(A=@n47.RGB, B=Multiply(A=OneMinus(None=@n48), B=LinearInterpolate(A=@n49, Alpha=@n50)))
  NamedRerouteDeclaration(None=Multiply(
    A=ChannelMaskParameter(None=@n51) {group=Super Level 24 Masks, parameterName=Primary Mask, sortPriority=5}.R
    B=ScalarParameter() {defaultValue=1, group=Super Level 24 Masks, parameterName=UseMask_Primary}
   )) {name=Primary Mask, nodeColor=<rgba 0.25 0 0.0124 1>}
  n52 = VectorParameter() {group=Super Level 24, parameterName=Sec_Color1, defaultValue=<rgba 0.3506 0 0.3841 1>}
  n53 = VectorParameter() {group=Super Level 24, parameterName=Sec_ColorFresnel, defaultValue=<rgba 0.0296 0.117 0.8879 1>}
  n54 = NamedRerouteUsage()
  n55 = NamedRerouteUsage()
  n56 = VectorParameter() {group=Super Level 24, parameterName=Sec_ColorGradientLocation, defaultValue=<rgba 0 0 100 1>}
  n57 = ScalarParameter() {defaultValue=50, group=Super Level 24, parameterName=Sec_ColorGradientRadius}
  n58 = ScalarParameter() {group=Super Level 24, parameterName=Sec_ColorGradientHardness}
  n59 = InverseLinearInterpolate(
   A=ScalarParameter() {group=Super Level 24, parameterName=Sec_ContrastRemapMin}
   B=ScalarParameter() {defaultValue=1, group=Super Level 24, parameterName=Sec_ContrastRemapMax}
   Value=Multiply(A=NamedRerouteUsage()) {constB=2}
  )
  n60 = OneMinus(None=NamedRerouteDeclaration(None=Saturate(None=@n59)) {name=GAO_Remapped, nodeColor=<rgba 0.6875 0 1 1>})
  n61 = ScalarParameter() {defaultValue=1, group=Super Level 24, parameterName=Sec_ContrastGAO_Exp}
  n62 = LinearInterpolate(
   A=LinearInterpolate(
    A=@n52.RGB
    B=VectorParameter() {group=Super Level 24, parameterName=Sec_Color2, defaultValue=<rgba 0.0166 0.0061 0.0692 1>}.RGB
    Alpha=SphereMask(A=@n55, B=@n56.RGB, Radius=@n57, Hardness=@n58)
   )
   B=@n53.RGB
   Alpha=Max(A=Power(Base=@n60, Exp=@n61), B=@n54)
  )
  n63 = VectorParameter() {group=Super Level 24, parameterName=Sec_SquaresColor, defaultValue=<rgba 0.9375 0.9375 0.9375 0>}
  n64 = NamedRerouteUsage()
  NamedRerouteDeclaration(None=StaticSwitchParameter(True=LinearInterpolate(A=@n52.RGB, B=@n53.RGB, Alpha=@n54), False=LinearInterpolate(A=@n62, B=@n63.RGB, Alpha=@n64)) {group=Super Level 24, parameterName=UseT3Pattern, sortPriority=3000}) {name=Secondary Color, nodeColor=<rgba 0 0.0186 1 1>}
  NamedRerouteDeclaration(None=NamedRerouteUsage()) {name=Emissive In, nodeColor=<rgba 0.0078 0 1 1>}
  n65 = Multiply(A=NamedRerouteUsage(), B=NamedRerouteUsage())
  n66 = StaticSwitchParameter(True=@n65, False=Add(A=@n65, B=NamedRerouteUsage()) {constB=0}) {parameterName=Use Toon Shading, sortPriority=900}
  n67 = Multiply(A=NamedRerouteUsage(), B=NamedRerouteUsage())
  n68 = NamedRerouteUsage()
  n69 = NamedRerouteUsage()
  NamedRerouteDeclaration(None=Add(A=Add(A=@n66, B=@n67), B=Multiply(A=@n68, B=@n69))) {name=Emissive Out, nodeColor=<rgba 1 0.8906 0 1>}
  n70 = NamedRerouteUsage()
  n71 = ScalarParameter() {defaultValue=0.55, parameterName=Sec_BoxSize}
  n72 = ScalarParameter() {defaultValue=0.05, group=Super Level 24, parameterName=Sec_Boxes_MaxThinkess}
  n73 = ScalarParameter() {parameterName=Sec_Boxes_Hardness}
  NamedRerouteDeclaration(None=Multiply(A=SphereMask(A=@n70, B=@n71, Radius=@n72, Hardness=@n73) {attenuationRadius=0.5}, B=NamedRerouteUsage())) {name=Sec_Boxes, nodeColor=<rgba 0 0.6484 1 1>}
  n74 = Add(
   A=ActorPositionWS()
   B=VectorParameter() {group=Super Level 24, parameterName=Sec_VirtualPlanPivotOffset1, defaultValue=<rgba 0 0 100 0>}.RGB
  )
  n75 = Add(A=ActorPositionWS(), B=VectorParameter() {group=Super Level 24, parameterName=Sec_VirtualPlanPivotOffset2}.RGB)
  n76 = NamedRerouteDeclaration(None=SphereMask(
    A=NamedRerouteUsage()
    B=VectorParameter() {group=Super Level 24, parameterName=Sec_VirtualPlaneSphereMaskLoc}.RGB
    Radius=ScalarParameter() {defaultValue=50, group=Super Level 24, parameterName=Sec_VirtualPlaneSphereMaskRadius}
   )) {name=Box Sphere Mask Debug, nodeColor=<rgba 0 0.1328 1 1>}
  n77 = NamedRerouteDeclaration(None=MaterialFunctionCall(
    Plane Normal Axis=MaterialFunctionCall() {materialFunction=&CameraDirectionVector}.Result
    Plane Center=LinearInterpolate(A=@n74, B=@n75, Alpha=@n76)
    UV Scale=ScalarParameter() {defaultValue=3, group=Super Level 24, parameterName=Sec_UVScale}
   ) {materialFunction=&VirtualPlaneCoordinates}.UVs) {name=Virtual_Plane, nodeColor=<rgba 0.4297 0 1 1>}
  n78 = TextureSample(UVs=@n77, Bias=ScalarParameter() {defaultValue=-4, group=Super Level 24, parameterName=DFTextureMipBias}) {
   mipValueMode=TMVM_MipBias
   samplerSource=SSM_Wrap_WorldGroupSettings
   samplerType=SAMPLERTYPE_LinearColor
   texture=&/T_ShapesDF
  }
  n79 = ConstantBiasScale(None=Sine(None=@n77))
  n80 = ComponentMask(None=@n79) {g=true}
  n81 = ComponentMask(None=@n79) {r=true}
  NamedRerouteDeclaration(None=ShadingPathSwitch(Default=@n78.R, Mobile=Max(A=@n80, B=@n81)) {desc=Migrated from FeatureLevelSwitch}) {name=Boxes_DF, nodeColor=<rgba 0.5469 1 0 1>}
  NamedRerouteDeclaration(None=FunctionInput() {bUsePreviewValueAsDefault=true, inputName=Skin, inputType=FunctionInput_Scalar, sortPriority=50}) {name=Skin, nodeColor=<rgba 0.8958 0.8958 0.8958 1>}
  n82 = Panner(
   Coordinate=Floor(None=NamedRerouteUsage())
   Speed=ComponentMask(None=VectorParameter() {parameterName=Sec_NoisePanningSpeed, defaultValue=<rgba 2 1 0 0>}.RGB) {g=true, r=true}
  )
  n83 = Add(
   A=TextureSample!(UVs=Divide(A=@n82) {constB=10}, Tex, Apply View MipBias) {samplerType=SAMPLERTYPE_Grayscale, texture=/Game/TimeOfDay/TODM_Sequence/Materials/Sky/Textures/LowResBlurredNoise_Radish_LF_Root}.R
   B=ScalarParameter() {defaultValue=-0.5, group=Super Level 24, parameterName=Sec_Boxes_Bias}
  )
  n84 = ScalarParameter() {defaultValue=2, group=Super Level 24, parameterName=Sec_BoxesMultiplier}
  NamedRerouteDeclaration(None=Saturate(None=Multiply(A=@n83, B=@n84))) {name=Noise, nodeColor=<rgba 0.9922 1 0 1>}
  n85 = FunctionInput() {bUsePreviewValueAsDefault=true, inputName=Inkline, inputType=FunctionInput_Scalar, sortPriority=500, previewValue=<1 0 0 1>}
  NamedRerouteDeclaration(None=@n85) {name=Inkline, nodeColor=<rgba 0.8958 0.8958 0.8958 1>}
  n86 = ChannelMaskParameter(None=@n51) {group=Super Level 24 Masks, maskChannel=Blue, parameterName=Secondary Mask, sortPriority=6, defaultValue=<rgba 0 0 1 0>}
  NamedRerouteDeclaration(None=Multiply(A=@n86.R, B=ScalarParameter() {defaultValue=1, group=Super Level 24 Masks, parameterName=UseMask_Secondary, sortPriority=33})) {name=Secondary Mask, nodeColor=<rgba 0 0.0046 0.25 1>}
  NamedRerouteDeclaration(None=MaterialFunctionCall(X=NamedRerouteUsage(), Y=NamedRerouteUsage(), Z=NamedRerouteUsage()) {materialFunction=&MakeFloat4}.Result) {name=Pattern, nodeColor=<rgba 0.7813 0 1 1>}
  n87 = OneMinus(None=NamedRerouteUsage())
  n88 = NamedRerouteUsage()
  n89 = ComponentMask(None=@n88) {r=true}
  n90 = MaterialFunctionCall(Y=Multiply(A=Abs(None=@n89), B=ScalarParameter() {defaultValue=0.3, parameterName=Pri_T3_SpiralStretch})) {materialFunction=&MakeFloat3}
  n91 = Multiply(A=Time(), B=ScalarParameter() {defaultValue=-1.5, parameterName=Pri_T3_BandsPanSpeed})
  n92 = Noise!(
   World Position=Add(
    A=Add(A=@n88, B=@n90.Result)
    B=MaterialFunctionCall(Z=@n91) {materialFunction=&MakeFloat3}.Result
   )
   FilterWidth
  ) {bTurbulence=false, levels=1, noiseFunction=NOISEFUNCTION_GradientALU, outputMin=0, scale=0.04}
  n93 = Add(
   A=Multiply(A=ComponentMask(None=@n92) {g=true}, B=ScalarParameter() {defaultValue=1, group=Dance Lines, parameterName=NoiseWarpIntensity})
   B=Multiply(A=Time(), B=ScalarParameter() {defaultValue=0.05, parameterName=Pri_T3_BandsSpeed})
  )
  n94 = CurveAtlasRowParameter(CurveTime=NamedRerouteDeclaration(None=Frac(None=@n93)) {name=NoiseStuff, nodeColor=<rgba 0.6875 0 1 1>}) {
   atlas=&/Atlast_SuperLevel24
   curve=&/Curve_SuperLevel24
   group=Super Level 24
   parameterName=Curve
  }
  NamedRerouteDeclaration(None=Multiply(A=@n87, B=@n94.B)) {name=T3 Scanline Falloff, nodeColor=<rgba 0.7344 0 1 1>}
  n95 = NamedRerouteUsage()
  n96 = NamedRerouteUsage()
  n97 = NamedRerouteUsage()
  n98 = Add(
   A=Multiply(
    A=VectorParameter() {group=Super Level 24, parameterName=Pri_GridLineCore, defaultValue=<rgba 0.7083 0.3136 0.6096 1>}.RGB
    B=ComponentMask(None=@n96) {r=true}
   )
   B=Multiply(
    A=VectorParameter() {group=Super Level 24, parameterName=Pri_GridLine, defaultValue=<rgba 0.5625 0 0.4219 1>}.RGB
    B=ComponentMask(None=@n97) {g=true}
   )
  )
  n99 = LinearInterpolate(B=ScalarParameter() {defaultValue=2, parameterName=Pri_WireBoostEdge}, Alpha=NamedRerouteUsage()) {constA=1}
  n100 = VectorParameter() {group=Super Level 24, parameterName=Pri_FresnelCenterColor, defaultValue=<rgba 0.0029 0 0.003 1>}
  n101 = VectorParameter() {group=Super Level 24, parameterName=Pri_FresnelEdgeColor, defaultValue=<rgba 0.1439 0.2964 1 1>}
  n102 = VectorParameter() {group=Super Level 24, parameterName=Pri_FresnelEdgeColor2, defaultValue=<rgba 0 0.35 0.1466 1>}
  n103 = Multiply(A=NamedRerouteUsage(), B=ScalarParameter() {defaultValue=0.1, parameterName=Pri_HueShiftMultiplier})
  n104 = Multiply(
   A=LinearInterpolate(A=@n100.RGB, B=LinearInterpolate(A=@n101.RGB, B=@n102.RGB, Alpha=@n103), Alpha=NamedRerouteUsage())
   B=ScalarParameter() {defaultValue=4, parameterName=Pri_HologramFresnelEdgeIntensity}
  )
  n105 = ComponentMask(None=NamedRerouteUsage()) {b=true}
  n106 = ScalarParameter() {defaultValue=0.5, parameterName=Pri_HologramFresnelCenterIntensity}
  n107 = Add(A=Add(A=Multiply(A=@n98, B=@n99), B=Multiply(A=@n104, B=@n105)) {constB=0}, B=Multiply(A=Multiply(A=@n100.RGB, B=@n106), B=@n105))
  NamedRerouteDeclaration(None=StaticSwitchParameter(True=Constant(), False=Multiply(A=@n107, B=@n95)) {group=Super Level 24, parameterName=UseT3Pattern, sortPriority=3000}) {name=Primary Emissive, nodeColor=<rgba 1 0 0.0496 1>}
  NamedRerouteDeclaration(None=@n5.Normal) {name=Normal In, nodeColor=<rgba 0.0078 0 1 1>}
  NamedRerouteDeclaration(None=NamedRerouteUsage()) {name=Normal Out, nodeColor=<rgba 1 0.8906 0 1>}
  n108 = FunctionInput() {bUsePreviewValueAsDefault=true, inputName=Toon Shading, sortPriority=300, previewValue=<1 0 0 1>}
  NamedRerouteDeclaration(None=ComponentMask(None=@n108) {r=true}) {name=Toon Shading Input, nodeColor=<rgba 1 0 0.5859 1>}
  n109 = ScalarParameter() {defaultValue=1, group=Super Level 24, parameterName=Pri_Metallic}
  n110 = LinearInterpolate(A=@n109, Alpha=NamedRerouteUsage())
  NamedRerouteDeclaration(None=StaticSwitchParameter(
    True=StaticSwitchParameter(True=@n109, False=@n110) {group=Super Level 24 Masks, parameterName=DisablePrimaryFX, sortPriority=3000}
    False=@n109
   ) {group=Super Level 24, parameterName=UseT3Pattern, sortPriority=3000}) {name=Primary Metallic, nodeColor=<rgba 1 0 0.0496 1>}
  n111 = NamedRerouteUsage()
  n112 = ScalarParameter() {defaultValue=-0.4, parameterName=Pri_T3_ScanLineCurve}
  n113 = Round(None=Add(A=Multiply(A=@n111, B=@n112) {constB=0.1}, B=@n94.R))
  n114 = OneMinus(None=NamedRerouteUsage())
  NamedRerouteDeclaration(None=Multiply(A=Multiply(A=@n114, B=@n113), B=@n87)) {name=T3 Bands, nodeColor=<rgba 0.2187 1 0 1>}
  n115 = ChannelMaskParameter(None=@n51) {group=Super Level 24 Masks, maskChannel=Green, parameterName=Accent Mask, sortPriority=7, defaultValue=<rgba 0 1 0 0>}
  NamedRerouteDeclaration(None=Multiply(A=@n115.R, B=ScalarParameter() {defaultValue=1, group=Super Level 24 Masks, parameterName=UseMask_Accent, sortPriority=34})) {name=Accent Mask, nodeColor=<rgba 0.0133 0.25 0 1>}
  n116 = FunctionInput() {bUsePreviewValueAsDefault=true, inputName=AO, inputType=FunctionInput_Scalar, sortPriority=30}
  n117 = FunctionInput() {bUsePreviewValueAsDefault=true, inputName=Gmap, inputType=FunctionInput_Scalar, sortPriority=40}
  NamedRerouteDeclaration(None=Multiply(
    A=NamedRerouteDeclaration(None=@n116) {name=AO, nodeColor=<rgba 1 0.3516 0 1>}
    B=NamedRerouteDeclaration(None=@n117) {name=Gmap, nodeColor=<rgba 1 0 0.6094 1>}
   )) {name=GAO, nodeColor=<rgba 0 1 1 1>}
  NamedRerouteDeclaration(None=FunctionInput(Preview=WorldPosition().XYZ) {bUsePreviewValueAsDefault=true, inputName=Preskinned Position, sortPriority=20}) {name=Preskinned Position, nodeColor=<rgba 0.5573 0 0.0376 1>}
  n118 = NamedRerouteUsage()
  NamedRerouteDeclaration(None=Multiply(A=VectorParameter() {group=Super Level 24, parameterName=Pnt_EmissiveColor, defaultValue=<rgba 0.0182 0 0.03 0>}.RGB, B=OneMinus(None=@n118))) {name=Paint Emissive, nodeColor=<rgba 0.4757 0.4315 0.5 1>}
  NamedRerouteDeclaration(None=@n5.EmissiveColor) {name=Emissive Pre, nodeColor=<rgba 0 0.4375 1 1>}
  n119 = Multiply(
   A=Multiply(A=TextureCoordinate(), B=ScalarParameter() {defaultValue=300, group=Super Level 24, parameterName=SL_PixelDensity})
   B=ScalarParameter() {defaultValue=1, parameterName=SL_PixelDensityObjectSize}
  )
  n120 = Multiply(A=TextureSample(UVs=NamedRerouteDeclaration(None=@n119) {name=TexCoordPixels, nodeColor=<rgba 0.125 1 0 1>}) {samplerType=SAMPLERTYPE_Masks, texture=/Game/Environments/Asteria/Sets/NeonCity/Techart/Textures/T_LCD_Panels_LCDSoft}.RGB) {constB=7.3728}
  n121 = ScalarParameter() {defaultValue=0.5, parameterName=Pri_PixelsBlend}
  NamedRerouteDeclaration(None=ShadingPathSwitch(Default=LinearInterpolate(B=@n120, Alpha=@n121) {constA=1, constAlpha=0.3}, Mobile=Constant() {r=1}) {desc=Migrated from FeatureLevelSwitch}) {name=Pixels, nodeColor=<rgba 0.8177 0.1235 0.7102 1>}
  NamedRerouteDeclaration(None=@n41.R) {name=Perminant Mask, nodeColor=<rgba 1 0.9609 0 1>}
  n122 = NamedRerouteUsage()
  n123 = ComponentMask(None=Desaturation(None=NamedRerouteUsage(), Fraction=Constant() {r=1})) {r=true}
  n124 = LinearInterpolate(
   A=VectorParameter() {group=Super Level 24 Masks, parameterName=InfernoColor1, defaultValue=<rgba 0.2159 0.0037 1 1>}.RGB
   B=VectorParameter() {group=Super Level 24 Masks, parameterName=InfernoColor2, defaultValue=<rgba 0.0578 0.807 1 1>}.RGB
   Alpha=SmoothStep(
    Min=ScalarParameter() {defaultValue=0.05, group=Super Level 24 Masks, parameterName=InfernoMin}
    Max=ScalarParameter() {defaultValue=0.3, group=Super Level 24 Masks, parameterName=InfernoMax}
    Value=@n123
   )
  )
  n125 = SmoothStep(
   Min=ScalarParameter() {group=Super Level 24 Masks, parameterName=InfernoMaskMin}
   Max=ScalarParameter() {defaultValue=0.1, group=Super Level 24 Masks, parameterName=InfernoMaskMax}
   Value=@n123
  )
  NamedRerouteDeclaration(None=StaticSwitchParameter(True=LinearInterpolate(A=@n122, B=@n124, Alpha=@n125), False=NamedRerouteUsage()) {group=Super Level 24 Masks, parameterName=UseInfernoAdjustOriginal}) {name=Emissive Post, nodeColor=<rgba 0 0.4375 1 1>}
  n126 = ChannelMaskParameter(None=@n51) {group=Super Level 24 Masks, maskChannel=Alpha, parameterName=Paint Mask, sortPriority=8, defaultValue=<rgba 0 0 0 1>}
  NamedRerouteDeclaration(None=Multiply(A=@n126.R, B=ScalarParameter() {defaultValue=1, group=Super Level 24 Masks, parameterName=UseMask_Paint, sortPriority=35})) {name=Paint Mask, nodeColor=<rgba 0.0951 0.0863 0.1 1>}
  n127 = Multiply(
   A=NamedRerouteUsage()
   B=LinearInterpolate(
    A=ScalarParameter() {defaultValue=0.01, group=Super Level 24 Toon, parameterName=Toon_ReflectionMin}
    B=ScalarParameter() {defaultValue=0.03, group=Super Level 24 Toon, parameterName=Toon_ReflectionMax}
    Alpha=NamedRerouteUsage()
   )
  )
  n128 = ScalarParameter() {group=Super Level 24 Toon, parameterName=Toon_MetallicReflectionMin}
  n129 = ScalarParameter() {defaultValue=0.06, group=Super Level 24 Toon, parameterName=Toon_MetallicReflectionMax}
  n130 = NamedRerouteUsage()
  n131 = Multiply(
   A=VectorParameter() {group=Super Level 24 Toon, parameterName=Toon_PriMetallicColor, defaultValue=<rgba 0.185 0 0.2865 1>}.RGB
   B=Multiply(A=NamedRerouteUsage(), B=LinearInterpolate(A=@n128, B=@n129, Alpha=@n130))
  )
  n132 = Transform(None=NamedRerouteUsage())
  n133 = SkyLightEnvMapSample(
   Direction=ReflectionVectorWS(CustomWorldNormal=@n132)
   Roughness=ScalarParameter() {defaultValue=0.1, group=Super Level 24, parameterName=Pri_AdditionalReflectionsRoughness}
  )
  n134 = Fresnel(ExponentIn=ScalarParameter() {defaultValue=6, group=Super Level 24, parameterName=Pri_AdditionalRefExponent}, Normal=@n132)
  n135 = ScalarParameter() {defaultValue=0.8, group=Super Level 24, parameterName=Pri_AdditionalRefMaxSpec}
  n136 = Saturate(None=NamedRerouteUsage())
  n137 = Multiply(
   A=Multiply(A=Multiply(A=@n133, B=@n134), B=LinearInterpolate(B=@n135, Alpha=@n136) {constA=0.1})
   B=ScalarParameter() {defaultValue=1, group=Super Level 24, parameterName=Pri_AdditionalRefMultiplier}
  )
  n138 = Constant()
  NamedRerouteDeclaration(None=StaticSwitchParameter(True=Add(A=@n127, B=@n131), False=ShadingPathSwitch(Default=@n137, Mobile=@n138) {desc=Migrated from FeatureLevelSwitch}) {parameterName=Use Toon Shading, sortPriority=900}) {name=AdditionalReflections, nodeColor=<rgba 0 1 0.8359 1>}
  n139 = NamedRerouteUsage()
  n140 = LinearInterpolate(
   A=ScalarParameter() {defaultValue=1, group=Super Level 24 Toon, parameterName=Toon_DiffuseMultiplier, sortPriority=900}
   B=ScalarParameter() {defaultValue=0.3, group=Super Level 24 Toon, parameterName=Toon_PriDiffuseMultiplier, sortPriority=900}
   Alpha=NamedRerouteUsage()
  )
  n141 = ScalarParameter() {defaultValue=1, group=Super Level 24 Toon, parameterName=Toon_SecDiffuseMultiplier, sortPriority=900}
  n142 = NamedRerouteUsage()
  n143 = Multiply(
   A=Multiply(A=NamedRerouteUsage(), B=LinearInterpolate(A=@n140, B=@n141, Alpha=@n142))
   B=LinearInterpolate(
    A=ScalarParameter() {defaultValue=0.5, group=Super Level 24 Toon, parameterName=Toon_MinBrightness, sortPriority=999}
    B=ScalarParameter() {defaultValue=1, group=Super Level 24 Toon, parameterName=Toon_MaxBrightness, sortPriority=999}
    Alpha=NamedRerouteUsage()
   )
  )
  n144 = LinearInterpolate(
   A=MaterialFunctionCall!(Color=Add(A=@n139, B=@n143)) {materialFunction=/Game/Packages/Fortress_SharedMaterials/Anime_Master_Material/MF_TODColorAdjustment}.Result
   B=NamedRerouteUsage()
   Alpha=NamedRerouteUsage()
  )
  n145 = NamedRerouteUsage()
  n146 = Add(A=Max(A=NamedRerouteUsage(), B=NamedRerouteUsage()), B=ScalarParameter() {defaultValue=0.35, parameterName=SL24_InklinesLift})
  NamedRerouteDeclaration(None=Multiply(A=Add(A=@n144, B=@n145), B=Saturate(None=@n146))) {name=Toon Shading, nodeColor=<rgba 1 0.1406 0 1>}
  n147 = ReflectionVectorWS()
  n148 = CollectionParameter() {
   collection=&FortniteMaterialParameters
   parameterName=SunlightDirectionalVector
  }
  n149 = ComponentMask(None=@n148) {b=true, g=true, r=true}
  n150 = ConstantBiasScale(None=NamedRerouteDeclaration(None=DotProduct(A=@n147, B=@n149)) {name=Reflection Vector Dot, nodeColor=<rgba 0 0.5078 1 1>})
  n151 = Add(A=@n150, B=ScalarParameter() {defaultValue=-0.4, parameterName=Toon_ReflectionBias})
  NamedRerouteDeclaration(None=Saturate(None=Round(None=@n151))) {name=Reflection Specular Dot Biased, nodeColor=<rgba 0.4531 0 1 1>}
  NamedRerouteDeclaration(None=ComponentMask(None=@n108) {g=true}) {name=Toon Specular, nodeColor=<rgba 1 0 0.75 1>}
  n152 = NamedRerouteUsage()
  n153 = StaticSwitchParameter(True=NamedRerouteUsage(), False=Constant() {r=1}) {group=Super Level 24, parameterName=UseT3Pattern, sortPriority=3000}
  n154 = NamedRerouteUsage()
  n155 = VectorParameter() {group=Super Level 24 Toon, parameterName=Toon_SecMetallicColor, defaultValue=<rgba 0.2346 0.5946 0.6719 1>}
  n156 = Add(
   A=Multiply(
    A=Multiply(A=@n153, B=@n154) {constA=1}
    B=VectorParameter() {group=Super Level 24 Toon, parameterName=Toon_PriMetallicColor, defaultValue=<rgba 0.185 0 0.2865 1>}.RGB
   )
   B=Multiply(A=NamedRerouteUsage(), B=@n155.RGB)
  )
  n157 = NamedRerouteUsage()
  NamedRerouteDeclaration(None=Multiply(A=Multiply(A=@n152, B=@n156), B=OneMinus(None=@n157))) {name=Toon Metallic, nodeColor=<rgba 1 0 0.0703 1>}
  n158 = TextureSample!(UVs, Tex=@n39, Bias=@n40, Apply View MipBias) {mipValueMode=TMVM_MipBias}
  n159 = ScalarParameter() {parameterName=Pri_T3_WireCoreMin}
  n160 = ScalarParameter() {defaultValue=0.25, parameterName=Pri_T3_WireCoreMax}
  n161 = Max(A=NamedRerouteDeclaration(None=@n158.R) {name=T3 Permenant Mask, nodeColor=<rgba 0 0.8828 1 1>}, B=NamedRerouteUsage())
  NamedRerouteDeclaration(None=SphereMask(
    A=@n158.G
    B=Constant()
    Radius=LinearInterpolate(A=@n159, B=@n160, Alpha=@n161)
    Hardness=ScalarParameter() {defaultValue=1, parameterName=Pri_T3_WireCoreHardness}
   )) {name=Wireframe Mask T3, nodeColor=<rgba 1 0.9609 0 1>}
  NamedRerouteDeclaration(None=SmoothStep(
    Min=ScalarParameter() {defaultValue=0.5, parameterName=Toon_MetallicReflectionBias}
    Max=ScalarParameter() {defaultValue=0.8, parameterName=Toon_MetallicReflectionBiass}
    Value=@n150
   )) {name=Reflection Metallic Dot Biased, nodeColor=<rgba 0.4531 0 1 1>}
  NamedRerouteDeclaration(None=Constant()) {name=Secondary Emissive, nodeColor=<rgba 0 0.0186 1 1>}
  NamedRerouteDeclaration(None=@n5.BaseColor) {name=Base Color Pre, nodeColor=<rgba 0 0.4375 1 1>}
  n162 = NamedRerouteUsage()
  NamedRerouteDeclaration(None=StaticSwitchParameter(True=LinearInterpolate(A=@n162, B=@n124, Alpha=@n125), False=NamedRerouteUsage()) {group=Super Level 24 Masks, parameterName=UseInfernoAdjustOriginal}) {name=Base Color Post, nodeColor=<rgba 0 0.4375 1 1>}
  n163 = ShadingPathSwitch(Default=ScalarParameter() {parameterName=Scanlines Dark}, Mobile=ScalarParameter() {defaultValue=0.8, parameterName=Scanlines Dark Mobile}) {desc=Migrated from FeatureLevelSwitch}
  n164 = Multiply(A=NamedRerouteUsage(), B=NamedRerouteUsage())
  n165 = ScreenPosition()
  n166 = ScalarParameter() {defaultValue=1000, parameterName=Pri_ScanlineScale}
  NamedRerouteDeclaration(None=MaterialFunctionCall!(
    Number of Scanlines=ScalarParameter() {defaultValue=150, parameterName=Scanline Amount}
    Scanline Speed
    "Scanlines Thickness (Bias)"=ScalarParameter() {defaultValue=-0.13, parameterName=ScanlineSize}
    Scanlines Dark Value=LinearInterpolate(A=@n163, Alpha=@n164)
    Scanlines Light Value=ScalarParameter() {defaultValue=1, parameterName=Scanlines Light}
    Block Opacity
    UVs=Divide(A=@n165.PixelPosition, B=@n166)
    Time
    Block Repeat
   ) {materialFunction=/Game/Effects/Fort_Effects/Effects/Atlas/Materials/MF_Atlas_Scanlines_01}.Result) {name=Scanlines, nodeColor=<rgba 0.5469 0 1 1>}
  n167 = NamedRerouteUsage()
  n168 = VectorParameter() {parameterName=Pri_NormalDirection, defaultValue=<rgba 1 0 0 1>}
  n169 = NamedRerouteUsage()
  n170 = Multiply(A=NamedRerouteUsage(), B=ConstantBiasScale(None=Sine(None=@n169)))
  n171 = ScalarParameter() {defaultValue=0.5, parameterName=Pri_NormalDirStrength}
  n172 = LinearInterpolate(
   A=@n167
   B=CrossProduct(A=@n167, B=Transform(None=@n168.RGB) {transformSourceType=TRANSFORMSOURCE_World, transformType=TRANSFORM_Tangent})
   Alpha=Multiply(A=Multiply(A=@n170, B=@n171), B=NamedRerouteUsage())
  )
  n173 = NamedRerouteUsage()
  NamedRerouteDeclaration(None=StaticSwitchParameter(True=@n167, False=StaticSwitchParameter(True=@n172, False=@n173) {group=Super Level 24, parameterName=UseT3Pattern, sortPriority=3000}) {group=Super Level 24 Masks, parameterName=DisablePrimaryFX, sortPriority=3000}) {name=Primary Normals, nodeColor=<rgba 1 0 0.0496 1>}
  n174 = NamedRerouteDeclaration(None=@n41.G) {name=Wireframe Texture, nodeColor=<rgba 1 0.9609 0 1>}
  n175 = ScalarParameter() {parameterName=Pri_Wire_SphereStart}
  n176 = Max(
   A=Power(Base=ConstantBiasScale(None=@n37), Exp=ScalarParameter() {defaultValue=2, group=Super Level 24, parameterName=expo})
   B=Multiply(A=NamedRerouteUsage(), B=ScalarParameter() {defaultValue=0.9, parameterName=Perminant Mask Multiplier}) {constB=0.3}
  )
  n177 = LinearInterpolate(A=ScalarParameter() {parameterName=Pri_WireCoreMin}, B=ScalarParameter() {defaultValue=0.3, parameterName=Pri_WireCoreMax}, Alpha=@n176)
  n178 = ScalarParameter() {defaultValue=0.2, parameterName=Pri_WireCoreHardness}
  NamedRerouteDeclaration(None=Multiply(A=SphereMask(A=@n174, B=@n175, Radius=@n177, Hardness=@n178), B=@n176)) {name=Wireframe Mask Core, nodeColor=<rgba 1 0.9609 0 1>}
  n179 = ScalarParameter() {defaultValue=0.4, group=Super Level 24, parameterName=Pri_Roughness}
  n180 = ScalarParameter() {defaultValue=0.25, parameterName=Pri_RoughnessBands}
  n181 = NamedRerouteUsage()
  n182 = ScalarParameter() {defaultValue=0.2, parameterName=Pri_AdditionalReflectionsRoughness}
  NamedRerouteDeclaration(None=StaticSwitchParameter(
    True=LinearInterpolate(A=@n179, B=@n180, Alpha=@n181)
    False=ShadingPathSwitch(Default=@n179, Mobile=@n182) {desc=Migrated from FeatureLevelSwitch}
   ) {group=Super Level 24, parameterName=UseT3Pattern, sortPriority=3000}) {name=Primary Roughness, nodeColor=<rgba 1 0 0.0496 1>}
  n183 = ScalarParameter() {defaultValue=1, group=Super Level 24, parameterName=Sec_ContrastFresnel_Exp}
  NamedRerouteDeclaration(None=SmoothStep(
    Min=ScalarParameter() {group=Super Level 24, parameterName=Sec_ContrastFresnelRemapMin}
    Max=ScalarParameter() {defaultValue=1, group=Super Level 24, parameterName=Sec_ContrastFresnelRemapMax}
    Value=Fresnel(ExponentIn=@n183) {baseReflectFraction=0}
   )) {name=Sec_Fresnel, nodeColor=<rgba 0.3811 0.5218 1 1>}
  n184 = Divide(A=NamedRerouteUsage(), B=ScalarParameter() {defaultValue=1, group=Super Level 24, parameterName=sineyScale})
  n185 = Multiply(
   A=Multiply(A=Time(), B=ScalarParameter() {defaultValue=10, group=Super Level 24, parameterName=sineySpeed})
   B=Add(
    A=MaterialFunctionCall() {materialFunction=&MakeFloat3}.Result
    B=VectorParameter() {parameterName=NoiseFlowDirection, defaultValue=<rgba 0.05 0.3 -1 1>}.RGB
   )
  )
  n186 = ComponentMask(None=MaterialFunctionCall(
    TextureObject=TextureObjectParameter() {parameterName=Pri_NoisePattern, texture=/Game/Global/Textures/Default/Noise/TankMoteNoise}
    TextureSize=ScalarParameter() {defaultValue=200, parameterName=Pri_NoiseSize}
    WorldPosition=Add(A=@n184, B=@n185)
    ProjectionTransitionContrast=Constant()
   ) {materialFunction=&WorldAlignedTexture}.XYZ Texture) {r=true}
  n187 = ScalarParameter() {defaultValue=1.3, parameterName=NoiseRemapMultiplier}
  n188 = Power(
   Base=Add(A=Multiply(A=@n186, B=@n187), B=ScalarParameter() {defaultValue=-0.3, parameterName=NoiseRemapBias})
   Exp=ScalarParameter() {defaultValue=2, parameterName=NoiseRemapPower}
  )
  n189 = ScalarParameter() {defaultValue=-0.01, parameterName=NoiseRemapOffset}
  NamedRerouteDeclaration(None=Saturate(None=Add(A=@n188, B=@n189))) {name=GeneratedNoise, desc=25 instructions, nodeColor=<rgba 1 0 0.1172 1>}
  n190 = CameraVectorWS()
  n191 = Constant3Vector() {constant=<rgba 1 1 -1 0>}
  NamedRerouteDeclaration(None=SkyLightEnvMapSample(Direction=Multiply(A=@n190, B=@n191), Roughness=Constant() {r=0.3})) {name=Fake Transparency, nodeColor=<rgba 0.9688 1 0 1>}
  NamedRerouteDeclaration(None=Fresnel(ExponentIn=ScalarParameter() {defaultValue=2.25, parameterName=FakeTransparencyFresnelExponent}) {baseReflectFraction=0}) {name=Fake Transparency Mask, nodeColor=<rgba 1 0.6328 0 1>}
  n192 = TextureCoordinate()
  n193 = ScalarParameter() {defaultValue=60, group=Super Level 24 Glitch UVs, parameterName=Block Amount}
  n194 = Round(None=Multiply(A=@n192, B=@n193))
  n195 = Add(A=NamedRerouteUsage(), B=ScalarParameter() {defaultValue=0.5, group=Super Level 24 Glitch UVs, parameterName=QuantizeNoiseBias})
  NamedRerouteDeclaration(None=LinearInterpolate(A=@n192, B=Divide(A=@n194, B=@n193), Alpha=Round(None=@n195))) {name=Tex Coord W/Glitching, desc=4 instructions, nodeColor=<rgba 1 0.2 0.2 1>}
  NamedRerouteDeclaration(None=StaticSwitchParameter(True=ScalarParameter() {defaultValue=1, group=Super Level 24, parameterName=Acc_T3_Metallic}, False=NamedRerouteUsage()) {group=Super Level 24, parameterName=UseT3Pattern, sortPriority=3000}) {name=Accent_MetallicOut, nodeColor=<rgba 0.0532 1 0 1>}
  n196 = NamedRerouteUsage()
  n197 = ScalarParameter() {defaultValue=20, parameterName=Glitch Size}
  n198 = MaterialFunctionCall(
   Seed=Divide(A=Time(), B=ScalarParameter() {defaultValue=5, parameterName=Glitching Speed}) {constB=10}
   Vec3=Ceil(None=Divide(A=@n196, B=@n197))
  ) {materialFunction=&MF_Hash}
  n199 = Subtract(A=ComponentMask(None=@n198.Vec3) {r=true}, B=ScalarParameter() {defaultValue=0.7, parameterName=Glitch Bias})
  NamedRerouteDeclaration(None=Multiply(A=@n198.Float, B=Saturate(None=@n199))) {name=Glitch Mask, desc=21 instructions, nodeColor=<rgba 0 0.5781 1 1>}
  n200 = ScalarParameter() {parameterName=Pri_WireGlowMin}
  n201 = ScalarParameter() {defaultValue=0.55, parameterName=Pri_WireGlowMax}
  n202 = SphereMask(
   A=@n174
   B=@n175
   Radius=Saturate(None=LinearInterpolate(A=@n200, B=@n201, Alpha=@n176))
   Hardness=ScalarParameter() {parameterName=Pri_WireGlowHardness}
  )
  n203 = ScalarParameter() {defaultValue=0.5, parameterName=Pri_WireGlowIntensity}
  NamedRerouteDeclaration(None=Multiply(A=Multiply(A=@n202, B=@n203), B=@n176)) {name=Wireframe Mask Glow, nodeColor=<rgba 1 0.9609 0 1>}
  n204 = ScalarParameter() {group=Super Level 24, parameterName=Pri_ContrastFresnelRemapMin}
  n205 = ScalarParameter() {defaultValue=1, group=Super Level 24, parameterName=Pri_ContrastFresnelRemapMax}
  n206 = CameraVectorWS()
  n207 = LinearInterpolate(
   A=PixelNormalWS()
   B=@n206
   Alpha=MaterialFunctionCall(Z=ScalarParameter() {defaultValue=0.3, parameterName=Pri_FresnelIgnoreZ}) {materialFunction=&MakeFloat3}.Result
  )
  n208 = Power(
   Base=OneMinus(None=DotProduct(A=@n207, B=@n206))
   Exp=ScalarParameter() {defaultValue=3.2, group=Super Level 24, parameterName=Pri_FresnelExponent}
  )
  NamedRerouteDeclaration(None=Saturate(None=InverseLinearInterpolate(A=@n204, B=@n205, Value=@n208))) {name=Fresnel Pri, nodeColor=<rgba 0.3811 0.5218 1 1>}
  NamedRerouteDeclaration(None=StaticSwitchParameter(True=ScalarParameter() {defaultValue=0.35, group=Super Level 24, parameterName=Acc_T3_Roughness}, False=NamedRerouteUsage()) {group=Super Level 24, parameterName=UseT3Pattern, sortPriority=3000}) {name=Accent_RoughnessOut, nodeColor=<rgba 0.0532 1 0 1>}
  NamedRerouteDeclaration(None=NamedRerouteUsage()) {name=Base Color In, nodeColor=<rgba 0.0078 0 1 1>}
  NamedRerouteDeclaration(None=@n94.G) {name=T3 Wireframe Blend, nodeColor=<rgba 0.7344 0 1 1>}
  NamedRerouteDeclaration(None=Multiply(A=@n1, B=ScalarParameter() {defaultValue=0.1, parameterName=T3 Scanlines Multiplier})) {name=T3 Scanline, nodeColor=<rgba 0.8 0.04 0.192 1>}
  NamedRerouteDeclaration(None=@n5.Metallic) {name=Metallic In, nodeColor=<rgba 0.0078 0 1 1>}
  NamedRerouteUsage()
 ]}
